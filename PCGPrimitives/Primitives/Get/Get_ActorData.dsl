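; Get_ActorData is a reusable PCG graph primitive that gathers data from all world actors matching the ActorTag string parameter.
; Get Actor Data selects multiple actors, takes data from their PCG components or parses their components, and always requeries without requiring overlap or bounds.
; The result passes through Delete Tags, which uses substring matching on "input" to clean data tags before output.
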(graph Get_ActorData
  (params (ActorTag :type String))
  (bind GetActorData_0 (Get%20Actor%20Data :ActorSelector (struct :ActorFilter AllWorldActors :bSelectMultiple true) :bAlwaysRequeryActors true :bComponentsMustOverlapSelf false :bIgnorePCGGeneratedComponents false :bTrackActorsOnlyWithinBounds false :Mode GetDataFromPCGComponentOrParseComponents :ActorSelectionTag ActorTag))
  (bind DeleteTags_0 (Delete%20Tags :Operator Substring :SelectedTags "input"))
  (connect GetActorData_0 Out DeleteTags_0 In)
  (output Out DeleteTags_0))
