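; Get_ActorDataByRef is a PCG graph primitive that fetches data from actors passed in through an array of soft actor references.
; A single Get Actor Data node takes the ActorReferences parameter as its input selector and sends the result to Out.
; It reads each actor's PCG component output or falls back to parsing its components, always requeries actors, and neither limits them to bounds nor ignores PCG-generated components.
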(graph Get_ActorDataByRef
  (params (ActorReferences :type (Array (SoftObject /Script/Engine.Actor))))
  (bind GetActorData_0 (Get%20Actor%20Data :ActorSelector (struct :ActorFilter FromInput :ActorReferenceSelector ActorReferences) :bAlwaysRequeryActors true :bComponentsMustOverlapSelf false :bIgnorePCGGeneratedComponents false :bTrackActorsOnlyWithinBounds false :Mode GetDataFromPCGComponentOrParseComponents :In ActorReferences))
  (output Out GetActorData_0))
