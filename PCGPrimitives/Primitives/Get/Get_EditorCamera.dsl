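; Get_EditorCamera is a PCG primitive graph that outputs data for the editor's perspective viewport cameras.
; It gathers all editor cameras, then keeps only elements whose ViewportType attribute equals "Perspective".
; The matching elements pass through the shared Primitives_PrepareOutputs template subgraph to the output, and the filter emits no data when nothing matches.
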(graph Get_EditorCamera
  (bind GetEditorCameras_0 (Get%20Editor%20Cameras))
  (bind AttributeFilter_0 (Filter%20Attribute%20Elements :AttributeTypes (struct :Type String :StringValue "Perspective") :bGenerateOutputDataEvenIfEmpty false :bUseConstantThreshold true :Operator Equal :TargetAttribute ViewportType :ThresholdAttribute @Last))
  ;[0 0 0 0.5 32] Prepare Output
  (bind Subgraph_0 (Subgraph :Graph /PCGPrimitives/Primitives/_Template_/Primitives_PrepareOutputs.Primitives_PrepareOutputs))
  (connect GetEditorCameras_0 Out AttributeFilter_0 In)
  (connect AttributeFilter_0 InsideFilter Subgraph_0 In)
  (output Out Subgraph_0))
